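# MetaSound patch that converts flight ability movement inputs into smoothed control values for audio.
# Speed, azimuth and elevation are range-mapped and interpolated, with azimuth made absolute, while an OnPlay-triggered ADSR outputs a fade-in envelope.
# SpeedDelta uses deferred variable reads to divide the change in unsmoothed normalized speed by the change in Time, then scales and smooths it.
# The Mobile page omits that delta calculation, leaving SpeedDelta unconnected.
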MetaSoundPatch /FortMetasound/Abilities/Flight/MSP_Ability_Flight_Control_Simple
inputs [
 AzimuthInA {type=Float}
 AzimuthInB {type=Float}
 AzimuthOutA {type=Float}
 AzimuthOutB {type=Float}
 FadeInTime {type=Time}
 OnPlay {type=Trigger}
 Speed {type=Float}
 SpeedAzimuth {type=Float}
 SpeedElevation {type=Float}
 SpeedFloor {type=Float}
 SpeedInterpTime {type=Time}
 Time {type=Float}
 TopSpeed {type=Float}
]
outputs [
 AzimuthSpeed {type=Float}
 ElevationSpeed {type=Float}
 FadeInTime {type=Float}
 NormalizedSpeed {type=Float}
 SpeedDelta {type=Float}
]
pages [
 Default {
  is_build_graph=true
  input_defaults {
   AzimuthInA=-20
   AzimuthInB=20
   AzimuthOutA=-1
   AzimuthOutB=1
   FadeInTime=1
   OnPlay=null
   Speed=0
   SpeedAzimuth=0
   SpeedElevation=0
   SpeedFloor=0
   SpeedInterpTime=0.5
   Time=0
   TopSpeed=1
  }
  nodes [
   ADSREnvelopeADSREnvelope_0 = ADSR Envelope.ADSR Envelope:Float(Attack Time=^FadeInTime, Trigger Attack=^OnPlay) {also {defaults {Sustain Level=1}}}
   AbsoluteValueAbs_0 = AbsoluteValue.Abs:Float(Input=UE.InterpTo:Audio(
     Interp Time=^SpeedInterpTime
     Target=MapRange.MapRange:Float(In Range A=^AzimuthInA, In Range B=^AzimuthInB, Out Range A=^AzimuthOutA, Out Range B=^AzimuthOutB, In=^SpeedAzimuth).Out Value
    ).Value)
   InterpTo_2 = UE.InterpTo:Audio(
    Interp Time=^SpeedInterpTime
    Target=MapRange.MapRange:Float(In=^SpeedElevation) {also {defaults {In Range A=-80, In Range B=80, Out Range A=1, Out Range B=-1}}}.Out Value
   )
   n1 = MapRange.MapRange:Float(In=^Speed, Out Range A=^SpeedFloor, In Range B=^TopSpeed)
   n2 = VariableDeferredAccessor.Float(Variable=InitVariable.Float().Variable)
   n3 = UE.Subtract:Float(
    PrimaryOperand=VariableAccessor.Float(Variable=VariableMutator.Float(Value=@n1.Out Value, Variable=@n2.Variable).Variable).Value
    AdditionalOperands=@n2.Value
   )
   n4 = VariableDeferredAccessor.Float(Variable=InitVariable.Float().Variable)
   n5 = UE.Subtract:Float(
    PrimaryOperand=VariableAccessor.Float(Variable=VariableMutator.Float(Value=^Time, Variable=@n4.Variable).Variable).Value
    AdditionalOperands=@n4.Value
   )
   InterpTo_3 = UE.InterpTo:Audio(Target=UE.Multiply:Float(PrimaryOperand=UE.Divide:Float(PrimaryOperand=@n3.Out, AdditionalOperands=@n5.Out).Out) {also {defaults {AdditionalOperands=1000}}}.Out) {also {defaults {Interp Time=1}}}
   InterpTo_1 = UE.InterpTo:Audio(Interp Time=^SpeedInterpTime, Target=@n1.Out Value)
   VariableMutator.Float(Variable=InitVariable.Float().Variable, Value=@InterpTo_1.Value)
  ]
  links [
   {dst="@FadeInTime", src=ADSREnvelopeADSREnvelope_0.Out Envelope}
   {dst="@AzimuthSpeed", src=AbsoluteValueAbs_0.Value}
   {dst="@NormalizedSpeed", src=InterpTo_1.Value}
   {dst="@ElevationSpeed", src=InterpTo_2.Value}
   {dst="@SpeedDelta", src=InterpTo_3.Value}
  ]
 }
 Mobile {
  nodes [
   ADSREnvelopeADSREnvelope_0 = ADSR Envelope.ADSR Envelope:Float(Attack Time=^FadeInTime, Trigger Attack=^OnPlay) {also {defaults {Sustain Level=1}}}
   AbsoluteValueAbs_0 = AbsoluteValue.Abs:Float(Input=UE.InterpTo:Audio(
     Interp Time=^SpeedInterpTime
     Target=MapRange.MapRange:Float(In Range A=^AzimuthInA, In Range B=^AzimuthInB, Out Range A=^AzimuthOutA, Out Range B=^AzimuthOutB, In=^SpeedAzimuth).Out Value
    ).Value)
   InitVariable.Float()
   InterpTo_2 = UE.InterpTo:Audio(
    Interp Time=^SpeedInterpTime
    Target=MapRange.MapRange:Float(In=^SpeedElevation) {also {defaults {In Range A=-80, In Range B=80, Out Range A=1, Out Range B=-1}}}.Out Value
   )
   n1 = MapRange.MapRange:Float(In=^Speed, Out Range A=^SpeedFloor, In Range B=^TopSpeed)
   InterpTo_1 = UE.InterpTo:Audio(Interp Time=^SpeedInterpTime, Target=@n1.Out Value)
   VariableMutator.Float(Variable=InitVariable.Float().Variable, Value=@InterpTo_1.Value)
   VariableMutator.Float(Variable=InitVariable.Float().Variable, Value=@n1.Out Value)
  ]
  links [
   {dst="@FadeInTime", src=ADSREnvelopeADSREnvelope_0.Out Envelope}
   {dst="@AzimuthSpeed", src=AbsoluteValueAbs_0.Value}
   {dst="@NormalizedSpeed", src=InterpTo_1.Value}
   {dst="@ElevationSpeed", src=InterpTo_2.Value}
  ]
 }
]
