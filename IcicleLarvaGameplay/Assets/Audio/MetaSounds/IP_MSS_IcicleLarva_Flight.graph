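# Looping stereo MetaSound source for Icicle Larva flight movement audio.
# A flight control patch turns speed and direction into values driving a traversal loop and a turn bank, while a distance gate triggers random pitch-varied passbys.
# Turn one-shots temporarily duck the traversal loop, and affiliation scales the final mix.
# The Mobile page mixes only the two flight layers in mono, without passbys or ducking.
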MetaSoundSource /IcicleLarvaGameplay/Assets/Audio/MetaSounds/IP_MSS_IcicleLarva_Flight
inputs [
 bBypassTopSpeedDebug {type=Bool}
 Character.Affiliation {type=Enum:FortAudioAffiliation}
 Character.Health {type=Float}
 Character.UniqueID {type=Int32}
 FadeInTime {type=Time}
 Movement.Speed {type=Float}
 Movement.SpeedAzimuth {type=Float}
 Movement.SpeedElevation {type=Float}
 Passby.OutputVolume {type=Float}
 Passby.Sounds {type=WaveAsset:Array}
 Passby.Threshold {type=Float}
 SpeedFloorNorm {type=Float}
 SpeedMax {type=Float}
 Time.ElapsedTime {type=Float}
 Time.GameTimeSeconds {type=Float}
 Time.PlayTime {type=Float}
 Time.TimeSinceLastPlayed {type=Float}
 Traversal.Loop.Sounds {type=WaveAsset:Array}
 Traversal.Loop.Volume {type=Float}
 Turn.Loop.Sounds {type=WaveAsset:Array}
 Turn.Loop.Volume {type=Float}
 Turn.OneShots.StartSounds {type=WaveAsset:Array}
 Turn.OneShots.VolumeControl {type=Float}
 Turn.OutputVolume {type=Float}
 Turn.RetriggerDelay {type=Time}
 Turn.Threshold {type=Float}
 UE.Attenuation.Distance {type=Float}
 UE.Source.OnPlay {type=Trigger}
 UE.Spatialization.Azimuth {type=Float}
 UE.Spatialization.Elevation {type=Float}
]
interfaces [Fort.Character:1.0, Fort.Movement:1.0, Fort.Time:1.0, UE.Attenuation:1.0, UE.OutputFormat.Stereo:1.0, UE.Source:1.1, UE.Spatialization:1.0]
outputs [
 "UE.OutputFormat.Stereo.Audio:0" {type=Audio}
 "UE.OutputFormat.Stereo.Audio:1" {type=Audio}
]
pages [
 Default {
  is_build_graph=true
  input_defaults {
   Character.Affiliation=0
   Character.Health=0
   Character.UniqueID=0
   FadeInTime=0
   Movement.Speed=0
   Movement.SpeedAzimuth=0
   Movement.SpeedElevation=0
   Passby.OutputVolume=1
   Passby.Sounds=null
   Passby.Threshold=3000
   SpeedFloorNorm=0
   SpeedMax=1000
   Time.ElapsedTime=0
   Time.GameTimeSeconds=0
   Time.PlayTime=0
   Time.TimeSinceLastPlayed=0
   Traversal.Loop.Volume=1
   Turn.Loop.Volume=1
   Turn.OneShots.VolumeControl=0.5
   Turn.OutputVolume=1
   Turn.RetriggerDelay=0.2
   Turn.Threshold=0.2
   UE.Attenuation.Distance=0
   UE.Source.OnPlay=null
   UE.Spatialization.Azimuth=0
   UE.Spatialization.Elevation=0
   bBypassTopSpeedDebug="<MetaSoundBuilderResult.SUCCEEDED: 0>"
   Traversal.Loop.Sounds []
   Turn.Loop.Sounds []
   Turn.OneShots.StartSounds []
  }
  nodes [
   n1 = /FortMetasound/Abilities/Flight/MSP_Ability_Flight_Control_Simple(
    FadeInTime=^FadeInTime
    Speed=^Movement.Speed
    SpeedElevation=^Movement.SpeedElevation
    SpeedFloor=^SpeedFloorNorm
    TopSpeed=^SpeedMax
    Time=^Time.PlayTime
    OnPlay=^UE.Source.OnPlay
    SpeedAzimuth=UE.Add:Float(PrimaryOperand=^UE.Spatialization.Azimuth) {also {defaults {AdditionalOperands=11}}}.Out
   )
   n2 = VariableAccessor.Float(Variable=VariableMutator.Float(Variable=InitVariable.Float().Variable, Value=@n1.NormalizedSpeed).Variable)
   n3 = VariableAccessor.Float(Variable=@n2.Variable)
   n4 = VariableAccessor.Float(Variable=VariableMutator.Float(Variable=InitVariable.Float().Variable, Value=@n1.AzimuthSpeed).Variable)
   n5 = VariableAccessor.Float(Variable=VariableMutator.Float(Variable=InitVariable.Float().Variable, Value=@n1.FadeInTime).Variable)
   n6 = /FortMetasound/Abilities/Flight/MSP_Ability_Flight_Bank(
    MovementSounds=^Turn.Loop.Sounds
    LoopVolume=^Turn.Loop.Volume
    OneShotSounds=^Turn.OneShots.StartSounds
    OneShotVolume=^Turn.OneShots.VolumeControl
    OneShotRetriggerDelay=^Turn.RetriggerDelay
    LoopThreshold=^Turn.Threshold
    OneshotThreshold=^Turn.Threshold
    NormSpeed=@n3.Value
    DirectionValue=@n4.Value
    SpeedVol=@n4.Value
    Fade=VariableAccessor.Float(Variable=@n5.Variable).Value
   )
   n7 = /FortMetasound/Abilities/Flight/MSP_Ability_Flight_Loop(
    MovementSounds=^Traversal.Loop.Sounds
    LoopVolume=^Traversal.Loop.Volume
    OnPlay=^UE.Source.OnPlay
    NormSpeed=UE.Multiply:Float(AdditionalOperands=VariableAccessor.Float(Variable=@n4.Variable).Value, PrimaryOperand=VariableAccessor.Float(Variable=@n3.Variable).Value).Out
    FadeIn=@n5.Value
    SpeedVol=@n2.Value
   )
   n8 = InitVariable.Float()
   n9 = VariableDeferredAccessor.Trigger(Variable=InitVariable.Trigger().Variable)
   TriggerAnyTriggerAny2_0 = "TriggerAny.Trigger Any (2)"(In 0=^UE.Source.OnPlay, In 1=@n9.Value)
   n10 = Value.Value:Float(Set=@n6.OnOneShotPlayed, Reset=@TriggerAnyTriggerAny2_0.Out) {also {defaults {Init Value=1, Target Value=0.6}}}
   n11 = UE.InterpTo:Audio(Target=@n10.Output Value) {also {defaults {Interp Time=0.5}}}
   n12 = UE.RandomFloat(Next=&MSP_OneshotGate(OneshotThreshold=^Passby.Threshold, Value=^UE.Attenuation.Distance) {also {defaults {EnterCondition=4, ExitCondition=3}}}.OnGateEnter) {also {defaults {Max=1.05, Min=0.95}}}
   ArrayRandomGet_0 = Array.Random Get:WaveAsset:Array(In Array=^Passby.Sounds, Next=@n12.On Next)
   FrequencyMultipliertoSemitone_0 = UE.Frequency Multiplier to Semitone:Float(Frequency Multiplier=@n12.Value)
   WavePlayer_0 = UE.Wave Player:Stereo(Play=@ArrayRandomGet_0.On Next, Wave Asset=@ArrayRandomGet_0.Value, Pitch Shift=@FrequencyMultipliertoSemitone_0.Semitones)
   AudioMixerAudioMixerStereo3_0 = "AudioMixer.Audio Mixer (Stereo, 3)"(
    Gain 1=^Passby.OutputVolume
    Gain 2=^Turn.OutputVolume
    In 2 L=@n6.Out L
    In 2 R=@n6.Out R
    In 0 L=@n7.Out L
    In 0 R=@n7.Out R
    Gain 0=VariableAccessor.Float(Variable=VariableMutator.Float(Variable=@n8.Variable, Value=@n11.Value).Variable).Value
    In 1 L=@WavePlayer_0.Out Left
    In 1 R=@WavePlayer_0.Out Right
   )
   n13 = &MSP_AffiliationMixFloat(Affiliation=^Character.Affiliation) {also {defaults {Local=0.9, Non-Enemy=0.4}}}
   Multiply_0 = UE.Multiply:Audio by Float(PrimaryOperand=@AudioMixerAudioMixerStereo3_0.Out R, AdditionalOperands=@n13.Out)
   Multiply_1 = UE.Multiply:Audio by Float(PrimaryOperand=@AudioMixerAudioMixerStereo3_0.Out L, AdditionalOperands=@n13.Out)
   /FortMetasound/Patches/PrintFloat(bBypass=^bBypassTopSpeedDebug) {also {defaults {Label=***}}}
   TriggerDelay_0 = UE.Trigger Delay(In=@n10.On Set) {also {defaults {Delay Time=1.5}}}
   VariableMutator.Trigger(Value=@TriggerDelay_0.Out, Variable=@n9.Variable)
  ]
  links [{dst="@UE.OutputFormat.Stereo.Audio:1", src=Multiply_0.Out}, {dst="@UE.OutputFormat.Stereo.Audio:0", src=Multiply_1.Out}]
 }
 Mobile {
  nodes [
   n1 = &MSP_Ability_Flight_Control_Simple(
    FadeInTime=^FadeInTime
    Speed=^Movement.Speed
    SpeedElevation=^Movement.SpeedElevation
    SpeedFloor=^SpeedFloorNorm
    TopSpeed=^SpeedMax
    Time=^Time.PlayTime
    OnPlay=^UE.Source.OnPlay
    SpeedAzimuth=UE.Add:Float(PrimaryOperand=^UE.Spatialization.Azimuth) {also {defaults {AdditionalOperands=11}}}.Out
   )
   n2 = VariableAccessor.Float(Variable=VariableMutator.Float(Variable=InitVariable.Float().Variable, Value=@n1.NormalizedSpeed).Variable)
   n3 = VariableAccessor.Float(Variable=@n2.Variable)
   n4 = VariableAccessor.Float(Variable=VariableMutator.Float(Variable=InitVariable.Float().Variable, Value=@n1.FadeInTime).Variable)
   AudioMixerAudioMixerMono2_0 = "AudioMixer.Audio Mixer (Mono, 2)"(
    Gain 1=^Turn.OutputVolume
    In 1=&MSP_Ability_Flight_Bank(
     MovementSounds=^Turn.Loop.Sounds
     LoopVolume=^Turn.Loop.Volume
     OneShotSounds=^Turn.OneShots.StartSounds
     OneShotVolume=^Turn.OneShots.VolumeControl
     OneShotRetriggerDelay=^Turn.RetriggerDelay
     LoopThreshold=^Turn.Threshold
     OneshotThreshold=^Turn.Threshold
     NormSpeed=@n3.Value
     SpeedVol=@n3.Value
     Fade=VariableAccessor.Float(Variable=@n4.Variable).Value
    ).Out L
    In 0=&MSP_Ability_Flight_Loop(
     MovementSounds=^Traversal.Loop.Sounds
     LoopVolume=^Traversal.Loop.Volume
     OnPlay=^UE.Source.OnPlay
     FadeIn=@n4.Value
     NormSpeed=@n2.Value
     SpeedVol=@n2.Value
    ).Out L
   )
   Multiply_0 = UE.Multiply:Audio by Float(
    PrimaryOperand=@AudioMixerAudioMixerMono2_0.Out
    AdditionalOperands=&MSP_AffiliationMixFloat(Affiliation=^Character.Affiliation) {also {defaults {In Value=0.71, Local=0.9, Non-Enemy=0.4}}}.Out
   )
   &PrintFloat(bBypass=^bBypassTopSpeedDebug) {also {defaults {Label=***}}}
   VariableMutator.Float(Variable=InitVariable.Float().Variable, Value=@n1.AzimuthSpeed)
  ]
  links [{dst="@UE.OutputFormat.Stereo.Audio:0", src=Multiply_0.Out}, {dst="@UE.OutputFormat.Stereo.Audio:1", src=Multiply_0.Out}]
 }
]
properties {
 assetImportData=&self:AssetImportData
 attenuationSettings=/IcicleLarvaGameplay/Assets/Audio/Settings/ATT_IcicleLarva_Flight_Movement
 bLooping=true
 duration=10000
 maxDistance=30500
 numChannels=2
 outputFormat=Stereo
 soundClassObject=&Fort_Gadgets_Items
 concurrencyOverrides {maxCount=2}
 concurrencySet [/IcicleLarvaGameplay/Assets/Audio/Settings/Laser_Eyes/CONC_IcicleLarva_Flight]
 modulationSettings {
  volumeRouting=Union
  volumeModulationDestination {bEnableModulation=true, modulators [/IcicleLarvaGameplay/Assets/Audio/Modulation/PP_JethroVehicle_Movement]}
 }
}
