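# MetaSound Source preset of IP_MSS_IcicleLarva_Flight that supplies the looping stereo sound for Icicle Larva sprint flight.
# The parent owns the graph; this preset feeds it a sprint traversal loop, four fly-by waves, and speed, passby and turn tuning inputs.
# Turn loop and one-shot wave arrays are empty.
# Flight-specific attenuation and concurrency apply, with an override allowing one instance that stops the oldest.
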MetaSoundSource /IcicleLarvaGameplay/Assets/Audio/MetaSounds/IP_MSS_IcicleLarva_Flight_Sprint_Preset
preset_parent=&/IP_MSS_IcicleLarva_Flight
inputs [
 bBypassTopSpeedDebug {type=Bool}
 Character.Affiliation {type=Enum:FortAudioAffiliation}
 Character.Health {type=Float}
 Character.UniqueID {type=Int32}
 FadeInTime {type=Time}
 Movement.Speed {type=Float}
 Movement.SpeedAzimuth {type=Float}
 Movement.SpeedElevation {type=Float}
 Passby.OutputVolume {type=Float}
 Passby.Sounds {type=WaveAsset:Array}
 Passby.Threshold {type=Float}
 SpeedFloorNorm {type=Float}
 SpeedMax {type=Float}
 Time.ElapsedTime {type=Float}
 Time.GameTimeSeconds {type=Float}
 Time.PlayTime {type=Float}
 Time.TimeSinceLastPlayed {type=Float}
 Traversal.Loop.Sounds {type=WaveAsset:Array}
 Traversal.Loop.Volume {type=Float}
 Turn.Loop.Sounds {type=WaveAsset:Array}
 Turn.Loop.Volume {type=Float}
 Turn.OneShots.StartSounds {type=WaveAsset:Array}
 Turn.OneShots.VolumeControl {type=Float}
 Turn.OutputVolume {type=Float}
 Turn.RetriggerDelay {type=Time}
 Turn.Threshold {type=Float}
 UE.Attenuation.Distance {type=Float}
 UE.Source.OnPlay {type=Trigger}
 UE.Spatialization.Azimuth {type=Float}
 UE.Spatialization.Elevation {type=Float}
]
interfaces [Fort.Character:1.0, Fort.Movement:1.0, Fort.Time:1.0, UE.Attenuation:1.0, UE.OutputFormat.Stereo:1.0, UE.Source:1.1, UE.Spatialization:1.0]
outputs [
 "UE.OutputFormat.Stereo.Audio:0" {type=Audio}
 "UE.OutputFormat.Stereo.Audio:1" {type=Audio}
]
pages [Default {input_defaults {
   Character.Affiliation=0
   Character.Health=0
   Character.UniqueID=0
   FadeInTime=0
   Movement.Speed=0
   Movement.SpeedAzimuth=0
   Movement.SpeedElevation=0
   Passby.OutputVolume=1
   Passby.Threshold=2000
   SpeedFloorNorm=0
   SpeedMax=2000
   Time.ElapsedTime=0
   Time.GameTimeSeconds=0
   Time.PlayTime=0
   Time.TimeSinceLastPlayed=0
   Traversal.Loop.Volume=0.7
   Turn.Loop.Volume=1
   Turn.OneShots.VolumeControl=0.5
   Turn.OutputVolume=1
   Turn.RetriggerDelay=0.2
   Turn.Threshold=0.2
   UE.Attenuation.Distance=0
   UE.Source.OnPlay=null
   UE.Spatialization.Azimuth=0
   UE.Spatialization.Elevation=0
   bBypassTopSpeedDebug="<MetaSoundBuilderResult.SUCCEEDED: 0>"
   Passby.Sounds [
    /IcicleLarvaGameplay/Assets/Audio/Waves/IP_IcicleLarva_FlyBy_01
    /IcicleLarvaGameplay/Assets/Audio/Waves/IP_IcicleLarva_FlyBy_02
    /IcicleLarvaGameplay/Assets/Audio/Waves/IP_IcicleLarva_FlyBy_03
    /IcicleLarvaGameplay/Assets/Audio/Waves/IP_IcicleLarva_FlyBy_04
   ]
   Traversal.Loop.Sounds [/IcicleLarvaGameplay/Assets/Audio/Waves/IP_MSS_IcicleLarva_Flight_Sprint_Loop]
   Turn.Loop.Sounds []
   Turn.OneShots.StartSounds []
  }}]
properties {
 assetImportData=&self:AssetImportData
 attenuationSettings=/IcicleLarvaGameplay/Assets/Audio/Settings/ATT_IcicleLarva_Flight_Movement
 bLooping=true
 duration=10000
 maxDistance=30500
 numChannels=2
 outputFormat=Stereo
 sampleRate=48000
 soundClassObject=&Fort_Gadgets_Items
 concurrencyOverrides {maxCount=1, resolutionRule=StopOldest}
 concurrencySet [/IcicleLarvaGameplay/Assets/Audio/Settings/Laser_Eyes/CONC_IcicleLarva_Flight]
}
