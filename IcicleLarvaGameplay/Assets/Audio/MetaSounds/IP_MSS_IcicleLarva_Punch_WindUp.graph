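# One-shot stereo MetaSound source for the Icicle Larva's punch wind-up sound, using the melee weapon sound class and punch attenuation.
# On play, two chained random floats roll pitch and gain variation, then trigger a random pick from the wind-up wave array.
# The pitch multiplier is converted to semitones, while both channels share the random gain plus a decibel Output Volume input.
# A concurrency override limits it to a single instance.
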MetaSoundSource /IcicleLarvaGameplay/Assets/Audio/MetaSounds/IP_MSS_IcicleLarva_Punch_WindUp
inputs [
 Output Volume {type=Float}
 Punch WindUp Sounds {type=WaveAsset:Array}
 UE.Source.OnPlay {type=Trigger}
]
interfaces [UE.OutputFormat.Stereo:1.0, UE.Source.OneShot:1.0, UE.Source:1.1]
outputs [
 "UE.OutputFormat.Stereo.Audio:0" {type=Audio}
 "UE.OutputFormat.Stereo.Audio:1" {type=Audio}
 UE.Source.OneShot.OnFinished {type=Trigger}
]
pages [Default {
  input_defaults {
   Output Volume=0
   UE.Source.OnPlay=null
   Punch WindUp Sounds [
    /IcicleLarvaGameplay/Assets/Audio/Waves/IP_IcicleLarva_Punch_WindUp_01
    /IcicleLarvaGameplay/Assets/Audio/Waves/IP_IcicleLarva_Punch_WindUp_02
    /IcicleLarvaGameplay/Assets/Audio/Waves/IP_IcicleLarva_Punch_WindUp_03
    /IcicleLarvaGameplay/Assets/Audio/Waves/IP_IcicleLarva_Punch_WindUp_04
   ]
  }
  nodes [
   DecibelstoLinearGain_0 = UE.Decibels to Linear Gain:Float(Decibels=^Output Volume)
   n1 = UE.RandomFloat(Next=^UE.Source.OnPlay) {also {defaults {Max=1.15, Min=0.95}}}
   n2 = UE.RandomFloat(Next=@n1.On Next) {also {defaults {Max=1.15, Min=0.95}}}
   n3 = UE.Reroute(Value=@n2.Value)
   ArrayRandomGet_0 = Array.Random Get:WaveAsset:Array(In Array=^Punch WindUp Sounds, Next=@n2.On Next)
   FrequencyMultipliertoSemitone_0 = UE.Frequency Multiplier to Semitone:Float(Frequency Multiplier=@n1.Value)
   WavePlayer_0 = UE.Wave Player:Stereo(
    Play=@ArrayRandomGet_0.On Next
    Wave Asset=@ArrayRandomGet_0.Value
    Pitch Shift=UE.Reroute(Value=@FrequencyMultipliertoSemitone_0.Semitones).Value
   )
   Multiply_0 = UE.Multiply:Audio by Float(
    AdditionalOperands=@DecibelstoLinearGain_0.Linear Gain
    PrimaryOperand=UE.Multiply:Audio by Float(AdditionalOperands=UE.Reroute(Value=@n3.Value).Value, PrimaryOperand=@WavePlayer_0.Out Left).Out
   )
   Multiply_2 = UE.Multiply:Audio by Float(
    AdditionalOperands=@DecibelstoLinearGain_0.Linear Gain
    PrimaryOperand=UE.Multiply:Audio by Float(AdditionalOperands=UE.Reroute(Value=@n2.Value).Value, PrimaryOperand=@WavePlayer_0.Out Right).Out
   )
  ]
  links [
   {dst="@UE.OutputFormat.Stereo.Audio:0", src=Multiply_0.Out}
   {dst="@UE.OutputFormat.Stereo.Audio:1", src=Multiply_2.Out}
   {dst="@UE.Source.OneShot.OnFinished", src=WavePlayer_0.On Finished}
  ]
 }]
properties {
 assetImportData=&self:AssetImportData
 attenuationSettings=/IcicleLarvaGameplay/Assets/Audio/Settings/ATT_IcicleLarva_Punch
 bOverrideConcurrency=true
 maxDistance=15000
 numChannels=2
 outputFormat=Stereo
 sampleRate=48000
 soundClassObject=&Fort_WeaponMelee
 concurrencyOverrides {maxCount=1}
}
