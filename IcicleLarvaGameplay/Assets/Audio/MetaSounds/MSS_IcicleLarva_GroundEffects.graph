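# Stereo MetaSound source for the Icicle Larva's ground effects, which switches between a ground wave and a water wave, with flight ground-effects attenuation.
# On play or any change of OnWaterHit, a compare starts the matching wave player.
# Each player's start attacks its own ADSR envelope and releases the other's, crossfading them in a mixer.
# Deferred trigger variables feed each envelope's On Done back to stop its faded-out player.
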MetaSoundSource /IcicleLarvaGameplay/Assets/Audio/MetaSounds/MSS_IcicleLarva_GroundEffects
inputs [
 OnWaterHit {type=Bool}
 UE.Source.OnPlay {type=Trigger}
]
interfaces [UE.OutputFormat.Stereo:1.0, UE.Source:1.1]
outputs [
 "UE.OutputFormat.Stereo.Audio:0" {type=Audio}
 "UE.OutputFormat.Stereo.Audio:1" {type=Audio}
]
pages [
 Default {
  is_build_graph=true
  input_defaults {OnWaterHit=null, UE.Source.OnPlay=null}
  nodes [
   TriggerOnValueChange_0 = UE.Trigger On Value Change:Bool(Value=^OnWaterHit)
   TriggerAnyTriggerAny2_0 = "TriggerAny.Trigger Any (2)"(In 0=^UE.Source.OnPlay, In 1=@TriggerOnValueChange_0.Trigger)
   n1 = TriggerCompare.Clamp:Bool(A=^OnWaterHit, Compare=@TriggerAnyTriggerAny2_0.Out) {also {defaults {B="<MetaSoundBuilderResult.SUCCEEDED: 0>"}}}
   n2 = VariableDeferredAccessor.Trigger(Variable=InitVariable.Trigger().Variable)
   WavePlayer_0 = UE.Wave Player:Stereo(Play=@n1.False, Stop=@n2.Value) {also {defaults {Loop="<MetaSoundBuilderResult.SUCCEEDED: 0>", Wave Asset=/IcicleLarvaGameplay/Assets/Audio/Waves/IcicleLarva_Ground_Effects}}}
   n3 = VariableDeferredAccessor.Trigger(Variable=InitVariable.Trigger().Variable)
   WavePlayer_1 = UE.Wave Player:Stereo(Play=@n1.True, Stop=@n3.Value) {also {defaults {Loop="<MetaSoundBuilderResult.SUCCEEDED: 0>", Wave Asset=/IcicleLarvaGameplay/Assets/Audio/Waves/IcicleLarva_Water_Effects}}}
   ADSREnvelopeADSREnvelope_0 = ADSR Envelope.ADSR Envelope:Float(Trigger Release=@WavePlayer_0.On Play, Trigger Attack=@WavePlayer_1.On Play) {also {defaults {Attack Time=1, Decay Time=1.5, Sustain Level=1}}}
   ADSREnvelopeADSREnvelope_1 = ADSR Envelope.ADSR Envelope:Float(Trigger Attack=@WavePlayer_0.On Play, Trigger Release=@WavePlayer_1.On Play) {also {defaults {Attack Time=1, Sustain Level=1}}}
   AudioMixerAudioMixerStereo2_0 = "AudioMixer.Audio Mixer (Stereo, 2)"(
    Gain 0=UE.Multiply:Float(PrimaryOperand=@ADSREnvelopeADSREnvelope_0.Out Envelope) {also {defaults {AdditionalOperands=0.6}}}.Out
    Gain 1=UE.Multiply:Float(PrimaryOperand=@ADSREnvelopeADSREnvelope_1.Out Envelope).Out
    In 1 L=@WavePlayer_0.Out Left
    In 1 R=@WavePlayer_0.Out Right
    In 0 L=@WavePlayer_1.Out Left
    In 0 R=@WavePlayer_1.Out Right
   )
   Multiply_0 = UE.Multiply:Audio by Float(PrimaryOperand=@AudioMixerAudioMixerStereo2_0.Out R) {also {defaults {AdditionalOperands=0.7}}}
   Multiply_1 = UE.Multiply:Audio by Float(PrimaryOperand=@AudioMixerAudioMixerStereo2_0.Out L) {also {defaults {AdditionalOperands=0.7}}}
   VariableMutator.Trigger(Value=@ADSREnvelopeADSREnvelope_0.On Done, Variable=@n3.Variable)
   VariableMutator.Trigger(Value=@ADSREnvelopeADSREnvelope_1.On Done, Variable=@n2.Variable)
  ]
  links [{dst="@UE.OutputFormat.Stereo.Audio:1", src=Multiply_0.Out}, {dst="@UE.OutputFormat.Stereo.Audio:0", src=Multiply_1.Out}]
 }
 Mobile {nodes [
   InitVariable.Trigger()
   InitVariable.Trigger()
   TriggerOnValueChange_0 = UE.Trigger On Value Change:Bool(Value=^OnWaterHit)
  ]}
]
properties {
 assetImportData=&self:AssetImportData
 attenuationSettings=/IcicleLarvaGameplay/Assets/Audio/Settings/ATT_IcicleLarva_Flight_GroundEffects
 bLooping=true
 bOverrideConcurrency=true
 duration=10000
 maxDistance=4100
 numChannels=2
 outputFormat=Stereo
 sampleRate=48000
 soundClassObject=&Fort_Gadgets_Items
 concurrencyOverrides {maxCount=1, resolutionRule=StopOldest}
}
